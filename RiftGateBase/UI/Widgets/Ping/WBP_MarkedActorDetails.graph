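# Details widget for a pinged marker, showing the marked actor's name, a diamond icon and a keybind prompt labelled Cancel Marker.
# SetMarkerData caches the marker and display info, sets the title, and async-loads the soft icon texture into the image's dynamic material.
# Outline, fill and separator colors are hardcoded red and grey, ignoring PrimaryColor.
# Construct sets the prompt text and stores the parent AthenaMarkerDetails, which is otherwise unused here.
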WidgetBlueprint /RiftGateBase/UI/Widgets/Ping/WBP_MarkedActorDetails
graphs [
 EventGraph {nodes [
   LoadObjectiveIcon() {is=title K2Node_CustomEvent}
   > n1 = Async Load Asset(Asset=Break Marked Actor Display Info!(MarkedActorDisplayInfo:"Marked Actor Display Info Structure (by ref)"=Get CachedDisplayInfo() -> (CachedDisplayInfo:Marked Actor Display Info Structure)) -> (
     DisplayName:Text
     Icon:Texture 2D Soft Object Reference
     CustomIndicatorClass:User Widget Soft Class Reference
     CustomDetailsClass:User Widget Soft Class Reference
     PrimaryColor:Linear Color Structure
     Sound:Sound Base Object Reference
     bAttachSound:bool=~
     ScreenClamping:EFortMarkedActorScreenClamping Enum
     SpecialCull:Bitmask Integer
     SpecialDotSizeThreshold:float=~
     SpecialDistance:float=~
     CustomTimeToLive:float=~
     AdditionalData:Marked Actor Display Info Additional Data Object Reference
    ) {is=title K2Node_BreakStruct}.Icon)
   n2 = Cast To Texture2D(execute=@n1.Completed, Object=@n1)
   > Update Objective Icon / Colors!(execute, self, DisplayedIcon:Texture Object Reference=@n2) -> (then)
   Event Construct()
   > "SetText (Text)"(self=Get Text_ActionText() -> (Text_ActionText:Common Text Object Reference), InText="LOCTEXT(\"Cancel Marker\")")
   > n3 = FindParentWidgetOfType!(
    execute
    StartingWidget:Widget Object Reference=GetParent!(self:Widget Object Reference) -> (ReturnValue:Panel Widget Object Reference)
    Type:Widget Class Reference=/Game/Athena/HUD/Team/AthenaMarkerDetails.AthenaMarkerDetails_C
   ) -> (then, ReturnValue:Athena Marker Details Object Reference)
   > n4 = Cast To AthenaMarkerDetails!(execute, Object=@n3) -> (then, CastFailed, AsAthena Marker Details:Athena Marker Details Object Reference) {is=title K2Node_DynamicCast}
   > Set AthenaMarkerDetailsParent(AthenaMarkerDetailsParent:Athena Marker Details Object Reference=@n4) -> (Output_Get:Athena Marker Details Object Reference)
   n5 = Event SetMarkerData!() -> (OutputDelegate, then, MarkerData:"Fort World Marker Data Structure (by ref)", DisplayInfo:"Marked Actor Display Info Structure (by ref)") {is=title K2Node_Event}
   > Set MarkerData(MarkerData:Fort World Marker Data Structure=@n5.MarkerData) -> (Output_Get:Fort World Marker Data Structure)
   > Set CachedDisplayInfo(CachedDisplayInfo:Marked Actor Display Info Structure=@n5.DisplayInfo) -> (Output_Get:Marked Actor Display Info Structure)
   > Update Objective Title()
   > LoadObjectiveIcon() {is=title K2Node_CallFunction}
  ]}
 Update Objective Icon / Colors {nodes [
   n1 = Update Objective Icon / Colors!() -> (then, DisplayedIcon:Texture Object Reference) {is=title K2Node_FunctionEntry}
   > n2 = GetDynamicMaterial(self=Get Image_ObjIcon() -> (Image_ObjIcon:Image Object Reference))
   > SetTextureParameterValue(self=@n2, ParameterName=Icon, Value=@n1)
   > SetVectorParameterValue(self=@n2, ParameterName=OutlineColor, Value=(R=1,G=0,B=0,A=1))
   > SetVectorParameterValue(self=@n2, ParameterName=FillColor, Value=(R=0.5711,G=0.5711,B=0.5711,A=1))
   > SetScalarParameterValue(self=@n2, ParameterName=Opacity, Value=1)
   > SetBrushColor(self=Get Border_Seperator() -> (Border_Seperator:Border Object Reference), InBrushColor=(R=1,G=0,B=0,A=1))
  ]}
 Update Objective Title {nodes [
   Update Objective Title() {is=title K2Node_FunctionEntry}
   > "SetText (Text)"(
    self=Get Text_ObjectiveName() -> (Text_ObjectiveName:Text Object Reference)
    InText=Break Marked Actor Display Info(MarkedActorDisplayInfo=Get CachedDisplayInfo()) -> (bAttachSound=~, SpecialDotSizeThreshold=~, SpecialDistance=~, CustomTimeToLive=~)
   )
  ]}
]
variables [
 AthenaMarkerDetailsParent {type=AthenaMarkerDetails_C, default=null}
 CachedDisplayInfo {
  type=MarkedActorDisplayInfo
  default {
   additionalData=null
   bAttachSound=true
   customDetailsClass=null
   customIndicatorClass=null
   customTimeToLive=-1
   displayName=""
   icon=null
   screenClamping=Default
   sound=null
   specialCull=0
   specialDistance=50
   specialDotSizeThreshold=-0.93
   primaryColor=<rgba 0 0 0 0>
  }
 }
 Color Reduction {type=number, default=0.08}
 MarkerData {
  type=FortWorldMarkerData
  default {
   bHeldMarkerInput=false
   bIncludeSquad=true
   bUseHoveredMarkerDetail=true
   itemCount=1
   itemDefinition=null
   markedActor=null
   markedActorClass=null
   markerType=None
   mostRecentArrayReplicationKey=-1
   owner=null
   pingCommand=Invalid
   pinnedQuestIndex=-1
   replicationId=-1
   replicationKey=-1
   targetConvertedType=Unknown
   basePosition=<0 0 0>
   basePositionOffset=<0 0 0>
   confirmationContainer {confirmations []}
   customDisplayInfos []
   markerId {instanceId=-1, playerId=-1}
   markerTag {tagName=None}
   worldNormal=<0 0 1>
  }
 }
]
widgets {
 info {inheritedWidgetCount=0, parentClass=&UserWidget, rootWidgetClass=&Border, widgetCount=13}
 widgets [
  ContentBox_Root {widgetClassPath=&Border, bIsVariable=true, widget=ContentBox_Root, properties {brushColor=<rgba 0 0 0 0>, padding=<ltrb 0 0 0 0>}}
  Overlay_Root {
   widgetClassPath=&Overlay
   parent=ContentBox_Root
   widget=Overlay_Root
   slot {class=BorderSlot, properties {padding=<ltrb 0 0 0 0>}}
  }
  VerticalBox_0 {
   widgetClassPath=&VerticalBox
   parent=Overlay_Root
   widget=VerticalBox_0
   slot {class=OverlaySlot, properties {horizontalAlignment=HAlign_Fill, verticalAlignment=VAlign_Fill}}
  }
  Text_ObjectiveName {
   widgetClassPath=&TextBlock
   bIsVariable=true
   parent=VerticalBox_0
   widget=Text_ObjectiveName
   properties {
    text="{MARKER TITLE}"
    font {
     bForceMonospaced=false
     bMaterialIsStencil=false
     fontMaterial=null
     fontObject=&Burbank
     letterSpacing=0
     monospacedWidth=1
     size=20
     skewAmount=0
     typefaceFontName=Burbank-PlayerPower
     outlineSettings {
      bApplyOutlineToDropShadows=true
      bMiteredCorners=false
      bSeparateFillAlpha=true
      outlineMaterial=&M-FontShadow
      outlineSize=3
      outlineColor=<rgba 0 0 0 0.8>
     }
    }
    renderTransform {angle=0, scale=<1 1>, shear=<-8 0>, translation=<0 0>}
   }
   slot {class=VerticalBoxSlot, properties {horizontalAlignment=HAlign_Left, verticalAlignment=VAlign_Bottom, padding=<ltrb 8 -5 0 -5>}}
  }
  Overlay_0 {
   widgetClassPath=&Overlay
   parent=VerticalBox_0
   widget=Overlay_0
   slot {class=VerticalBoxSlot, properties {padding=<ltrb 0 4 0 0>}}
  }
  Border_Seperator {
   widgetClassPath=&Border
   bIsVariable=true
   parent=Overlay_0
   widget=Border_Seperator
   properties {
    background {
     bIsDynamicallyLoaded=false
     drawAs=Image
     imageType=NoImage
     mirroring=NoMirror
     resourceName=None
     resourceObject=null
     tiling=NoTile
     imageSize=<32 16>
     margin=<ltrb 0 0 0 0>
     outlineSettings {
      bUseBrushTransparency=false
      roundingType=HalfHeightRadius
      width=0
      color {colorUseRule=UseColor_Specified, specifiedColor=<rgba 0 0 0 0>}
      cornerRadii=<0 0 0 0>
     }
     tintColor {colorUseRule=UseColor_Specified, specifiedColor=<rgba 1 1 1 1>}
     uVRegion {bIsValid=false, max=<0 0>, min=<0 0>}
    }
    brushColor=<rgba 1 0 0 1>
    padding=<ltrb 4 0 4 3>
    renderTransformPivot=<0 0.5>
   }
   slot {class=OverlaySlot, properties {horizontalAlignment=HAlign_Fill, verticalAlignment=VAlign_Center, padding=<ltrb 3 0 0 0>}}
  }
  FortVisualAttachment {
   widgetClassPath=&FortVisualAttachment
   parent=Overlay_0
   widget=FortVisualAttachment
   properties {widthOverride=100}
   slot {class=OverlaySlot, properties {verticalAlignment=VAlign_Center}}
  }
  Overlay_ObjIcon {
   widgetClassPath=&Overlay
   parent=FortVisualAttachment
   widget=Overlay_ObjIcon
   slot {class=SizeBoxSlot, properties {horizontalAlignment=HAlign_Right, verticalAlignment=VAlign_Top}}
  }
  Image_ObjIcon {
   widgetClassPath=&Image
   bIsVariable=true
   parent=Overlay_ObjIcon
   widget=Image_ObjIcon
   properties {
    brush {
     bIsDynamicallyLoaded=false
     drawAs=Image
     imageType=NoImage
     mirroring=NoMirror
     resourceName=None
     resourceObject=/Game/Athena/HUD/Materials/M_UI_PingDiamond_Inst
     tiling=NoTile
     imageSize=<86 86>
     margin=<ltrb 0 0 0 0>
     outlineSettings {
      bUseBrushTransparency=false
      roundingType=HalfHeightRadius
      width=0
      color {colorUseRule=UseColor_Specified, specifiedColor=<rgba 0 0 0 0>}
      cornerRadii=<0 0 0 0>
     }
     tintColor {colorUseRule=UseColor_Specified, specifiedColor=<rgba 1 1 1 1>}
     uVRegion {bIsValid=false, max=<0 0>, min=<0 0>}
    }
    renderTransform {angle=0, scale=<1 1>, shear=<0 0>, translation=<-73 -43>}
    renderTransformPivot=<1 0.5>
   }
   slot {class=OverlaySlot, properties {horizontalAlignment=HAlign_Fill, verticalAlignment=VAlign_Fill}}
  }
  HBox_ActionTextContainer {
   widgetClassPath=&HorizontalBox
   bIsVariable=true
   parent=VerticalBox_0
   widget=HBox_ActionTextContainer
   slot {class=VerticalBoxSlot, properties {padding=<ltrb 14 2 8 0>}}
  }
  ScaleBox_0 {
   widgetClassPath=&ScaleBox
   parent=HBox_ActionTextContainer
   widget=ScaleBox_0
   properties {stretch=UserSpecified, userSpecifiedScale=0.6}
   slot {class=HorizontalBoxSlot, properties {verticalAlignment=VAlign_Center}}
  }
  Keybind_MapMarker {
   widgetClassPath=/FortUILibrary/Input/KeybindWidget
   bIsVariable=true
   parent=ScaleBox_0
   widget=Keybind_MapMarker
   properties {boundAction=PlaceMapMarkerInWorld}
   slot {class=ScaleBoxSlot}
  }
  Text_ActionText {
   widgetClassPath=&CommonTextBlock
   bIsVariable=true
   parent=HBox_ActionTextContainer
   widget=Text_ActionText
   properties {
    text=Text Block
    textTransformPolicy=ToUpper
    font {
     bForceMonospaced=false
     bMaterialIsStencil=false
     fontMaterial=null
     fontObject=&BurbankSmall
     letterSpacing=0
     monospacedWidth=1
     size=14
     skewAmount=0
     typefaceFontName=Bold
     outlineSettings {
      bApplyOutlineToDropShadows=true
      bMiteredCorners=false
      bSeparateFillAlpha=true
      outlineMaterial=&M-FontShadow
      outlineSize=3
      outlineColor=<rgba 0 0 0 0.5>
     }
    }
    shadowOffset=<0 0>
   }
   slot {class=HorizontalBoxSlot, properties {verticalAlignment=VAlign_Center, padding=<ltrb 4 -2 0 -2>}}
  }
 ]
}
